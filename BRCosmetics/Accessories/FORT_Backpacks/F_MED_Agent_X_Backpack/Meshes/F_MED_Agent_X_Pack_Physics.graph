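# Physics asset for the Agent X backpack cosmetic, defining simple collision for the pack and its antenna.
# A box covers the pack, a small box the antenna base, and a capsule the antenna parent bone.
# One constraint links the antenna parent to its base with narrow swing and tighter twist limits, keeping any antenna motion small.
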PhysicsAsset /BRCosmetics/Accessories/FORT_Backpacks/F_MED_Agent_X_Backpack/Meshes/F_MED_Agent_X_Pack_Physics
bodies {
 dyn_antenna_base [{shapeType=Box, extentX=0.75, extentY=0.75, extentZ=0.75}]
 dyn_antenna_parent [{shapeType=Capsule, length=2.5, radius=1.5, center=<2.3449 0 -0.0001>, rotation=<rot 90 0 0>}]
 pack [{shapeType=Box, extentX=7.5, extentY=7.5, extentZ=7.5}]
}
constraints [{bone1Name=dyn_antenna_parent, bone2Name=dyn_antenna_base, swing1LimitDegrees=15, swing2LimitDegrees=15, twistLimitDegrees=5}]
