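# Animation Blueprint for the Agent X medium backpack cosmetic, adding secondary physics motion to its skeleton.
# The input pose has dyn_antenna_base constrained to the pack bone, then runs through a RigidBody node using the pack's physics asset.
# Simulation is relative to the root bone, with overridden gravity and scaled-down component motion.
# An Anim Dynamics node for dyn_antenna_parent is left unconnected, and the event graph holds no logic.
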AnimBlueprint /BRCosmetics/Accessories/FORT_Backpacks/F_MED_Agent_X_Backpack/Meshes/F_MED_Agent_X_Pack_AnimBp
target_skeleton=&/F_MED_Agent_X_Pack_Skeleton
graphs [
 AnimGraph {nodes [
   Anim Dynamics - Bone: dyn_antenna_parent(Alpha=0.5, ExternalForce="0,0,0") {
    angularDampingOverride=0.99
    bGravityOverrideInSimSpace=true
    bOverrideAngularDamping=true
    bOverrideLinearDamping=true
    bUseGravityOverride=true
    gravityScale=1.5
    linearDampingOverride=0.99
    alphaBoolBlend {blendOption=Linear}
    boundBone {boneName=dyn_antenna_parent}
    componentAppliedLinearAccClamp=<6000 6000 6000>
    componentLinearAccScale=<0.5 0.5 0.5>
    componentLinearVelScale=<0.5 0.5 0.5>
    gravityOverride=<1 1 -90>
    physicsBodyDefinitions [{
      sphereCollisionRadius=0
      boundBone {boneName=dyn_antenna_parent}
      boxExtents=<4 8 8>
      constraintSetup {angularLimitsMax=<0 10 10>, angularLimitsMin=<0 -10 -20>, angularTarget=<0 0 0>}
      localJointOffset=<-1 0 -1>
     }]
    relativeSpaceBone {boneName=root}
   }
   n1 = "Constraint - dyn_antenna_base (P)"!(ComponentPose=Local To Component(LocalPose=Input Pose()), Alpha=1, ConstraintWeights_0=1, ConstraintWeights_1:float=0) -> (Pose) {
    alphaBoolBlend {blendOption=Linear}
    boneToModify {boneName=dyn_antenna_base}
    constraintSetup [{targetBone {boneName=pack}}]
    constraintWeights [1,0]
   }
   n2 = RigidBody(ComponentPose=@n1) {
    bClampLinearTranslationLimitToRefPose=true
    bOverrideWorldGravity=true
    overridePhysicsAsset=&/F_MED_Agent_X_Pack_Physics
    simulationSpace=BaseBoneSpace
    alphaBoolBlend {blendOption=Linear}
    baseBoneRef {boneName=root}
    componentLinearAccScale=<0.4 0.4 0.4>
    componentLinearVelScale=<0.4 0.4 0.4>
    overrideWorldGravity {z=-980}
   }
   Output Pose(Result=Component To Local(ComponentPose=@n2)) {layerGroup=None}
  ]}
 EventGraph {nodes [
   TryGetPawnOwner()
   Event BlueprintUpdateAnimation()
  ]}
]
variables []
